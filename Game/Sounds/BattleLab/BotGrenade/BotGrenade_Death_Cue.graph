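# Sound cue for the BotGrenade death sound in BattleLab, classed under gadget items.
# A short fixed delay feeds a modulator for pitch and volume variation, then an equally weighted random pick among three death waves.
# It plays at low volume through the gadget teleporter attenuation asset, with overridden concurrency that prevents new instances once the limit is reached.
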SoundCue /Game/Sounds/BattleLab/BotGrenade/BotGrenade_Death_Cue
root=Delay_0
nodes [
 {id=Delay_0, children [Modulator_4], properties {delayMax=0.1, delayMin=0.1}}
 {id=Modulator_4, children [Random_4]}
 {id=Random_4, children [WavePlayer_8, WavePlayer_9, WavePlayer_10], properties {weights [1,1,1]}}
 {id=WavePlayer_10, properties {soundWaveAssetPtr=&/BotGrenade_Death_B03}}
 {id=WavePlayer_8, properties {soundWaveAssetPtr=&/BotGrenade_Death_B01}}
 {id=WavePlayer_9, properties {soundWaveAssetPtr=&/BotGrenade_Death_B02}}
]
properties {
 attenuationSettings=/Game/Sounds/Attenuation/Fort_GadgetTeleporter_Att
 bOverrideConcurrency=true
 maxDistance=4500
 soundClassObject=&Fort_Gadgets_Items
 volumeMultiplier=0.16
 attenuationOverrides {falloffDistance=1500, attenuationShapeExtents {x=800}}
 concurrencyOverrides {resolutionRule=PreventNew}
}
