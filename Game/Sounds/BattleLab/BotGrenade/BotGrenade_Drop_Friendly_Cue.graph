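# Sound cue for dropping a friendly Bot Grenade in Battle Lab, reusing existing weapon and trap sounds.
# A Random node picks one of two healing grenade drop waves, and a Mixer layers it with a poison dart trap drop wave.
# A Modulator feeds the output, which plays quieter and higher-pitched using weapon drop attenuation and the loot collect sound class.
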SoundCue /Game/Sounds/BattleLab/BotGrenade/BotGrenade_Drop_Friendly_Cue
root=Modulator_0
nodes [
 {id=Mixer_0, children [Random_1,WavePlayer_4], properties {inputVolume [1,1]}}
 {id=Modulator_0, children [Mixer_0]}
 {id=Random_1, children [WavePlayer_0,WavePlayer_3], properties {weights [1,1]}}
 {id=WavePlayer_0, properties {soundWaveAssetPtr=/Game/Sounds/Fort_Weapons/Launchers/Healing/HealingGrenade_Drop_02}}
 {id=WavePlayer_3, properties {soundWaveAssetPtr=/Game/Sounds/Fort_Weapons/Launchers/Healing/HealingGrenade_Drop_01}}
 {id=WavePlayer_4, properties {soundWaveAssetPtr=/Game/Sounds/Fort_Traps/PoisonDart_Trap/PoisonDartTrap_Drop_01}}
]
properties {
 attenuationSettings=&Fort_Weapon_Drop_Att
 maxDistance=3000
 pitchMultiplier=1.2
 soundClassObject=&Fort_Loot_Collect
 volumeMultiplier=0.2
}
