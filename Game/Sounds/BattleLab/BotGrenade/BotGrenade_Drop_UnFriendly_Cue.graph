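# Sound cue for the drop sound of the BattleLab unfriendly bot grenade.
# A Mixer layers two borrowed drop waves, a quieter high-infantry auto rifle drop and a louder poison dart trap drop.
# The mix feeds a default Modulator for variation, with the cue slightly pitched up and lowered in volume.
# It uses weapon-drop attenuation and the loot-collect sound class.
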SoundCue /Game/Sounds/BattleLab/BotGrenade/BotGrenade_Drop_UnFriendly_Cue
root=Modulator_0
nodes [
 {id=Mixer_0, children [WavePlayer_0,WavePlayer_3], properties {inputVolume [0.5,1.1]}}
 {id=Modulator_0, children [Mixer_0]}
 {id=WavePlayer_0, properties {soundWaveAssetPtr=/Game/Sounds/Fort_Weapons/AutoRifle/HighInfantry/HiInfantry_Drop_01}}
 {id=WavePlayer_3, properties {soundWaveAssetPtr=/Game/Sounds/Fort_Traps/PoisonDart_Trap/PoisonDartTrap_Drop_01}}
]
properties {
 attenuationSettings=&Fort_Weapon_Drop_Att
 maxDistance=3000
 pitchMultiplier=1.1
 soundClassObject=&Fort_Loot_Collect
 volumeMultiplier=0.4
}
