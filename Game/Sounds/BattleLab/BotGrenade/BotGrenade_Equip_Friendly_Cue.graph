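# Sound cue for the friendly BotGrenade equip sound in BattleLab.
# A single wave player feeds the BotGrenade_Equip_Friendly_01 wave straight to the output, with a slightly lowered pitch and boosted volume.
# It routes to the player foley sound class with close-range prop attenuation, and its concurrency override blocks new instances rather than stopping playing ones.
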SoundCue /Game/Sounds/BattleLab/BotGrenade/BotGrenade_Equip_Friendly_Cue
root=WavePlayer_0
nodes [{id=WavePlayer_0, properties {soundWaveAssetPtr=&/BotGrenade_Equip_Friendly_01}}]
properties {
 attenuationSettings=/Game/Sounds/Attenuation/Fort_Prop_Close_Att
 bOverrideConcurrency=true
 maxDistance=1800
 pitchMultiplier=0.9
 soundClassObject=&Fort_Player_Foley
 volumeMultiplier=1.4
 concurrencyOverrides {resolutionRule=PreventNew}
}
