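# Sound cue for the looping in-air sound of an unfriendly BattleLab bot grenade.
# A single looping wave player feeds the output at reduced volume under the ranged-weapon projectile effects sound class.
# Spatialization references the shared projectile in-air loop attenuation asset.
# Overridden concurrency blocks new instances when the limit is reached, leaving playing loops uninterrupted.
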SoundCue /Game/Sounds/BattleLab/BotGrenade/BotGrenade_InAirLoop_UnFriendly_Cue
root=WavePlayer_0
nodes [{id=WavePlayer_0, properties {bLooping=true, soundWaveAssetPtr=&/BotGrenade_InAirLoop_UnFriendly_02}}]
properties {
 attenuationSettings=/Game/Sounds/Attenuation/Fort_Weapon_Projectile_InAir_Loop_Att
 bOverrideConcurrency=true
 maxDistance=4500
 soundClassObject=&Fort_WeaponRanged_Projectile_Effects
 volumeMultiplier=0.45
 attenuationOverrides {falloffDistance=5000, reverbWetLevelMax=0.7, reverbWetLevelMin=0.2, attenuationShapeExtents {x=1000}}
 concurrencyOverrides {resolutionRule=PreventNew}
}
