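# Sound cue for the friendly Bot Grenade pickup in BattleLab, classed under gadget items.
# A single wave player feeds one pickup sound straight to the output, with slightly raised pitch and reduced volume.
# It uses the shared weapon pickup attenuation and overrides concurrency to block new instances rather than cut off playing ones.
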SoundCue /Game/Sounds/BattleLab/BotGrenade/BotGrenade_Pickup_Friendly_Cue
root=WavePlayer_1
nodes [{id=WavePlayer_1, properties {soundWaveAssetPtr=&/BotGrenade_Pickup_Friendly_01}}]
properties {
 attenuationSettings=&Fort_Weapon_Pickup_Att
 bOverrideConcurrency=true
 maxDistance=3500
 pitchMultiplier=1.1
 soundClassObject=&Fort_Gadgets_Items
 volumeMultiplier=0.45
 concurrencyOverrides {resolutionRule=PreventNew}
}
