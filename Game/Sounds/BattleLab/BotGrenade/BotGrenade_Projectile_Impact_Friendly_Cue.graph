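# Sound cue for the friendly BattleLab BotGrenade projectile impact, routed to the ranged weapon projectile effects sound class.
# A mixer layers a randomly picked, pitch-raised generic grenade bounce with a dedicated friendly impact wave that starts after a short fixed delay.
# Inline attenuation with distance low-pass filtering overrides the referenced grenade bounce attenuation asset, and a concurrency override caps simultaneous instances.
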SoundCue /Game/Sounds/BattleLab/BotGrenade/BotGrenade_Projectile_Impact_Friendly_Cue
root=Mixer_0
nodes [
 {id=Delay_0, children [WavePlayer_0], properties {delayMax=0.05, delayMin=0.05}}
 {id=Mixer_0, children [Modulator_0,Delay_0], properties {inputVolume [1,1]}}
 {id=Modulator_0, children [Random_4], properties {pitchMax=1.2, pitchMin=1.1}}
 {id=Random_4, children [WavePlayer_28, WavePlayer_29], properties {weights [1,1]}}
 {id=WavePlayer_0, properties {soundWaveAssetPtr=&/BotGrenade_Projectile_Impact_Friendly_01}}
 {id=WavePlayer_28, properties {soundWaveAssetPtr=/Game/Sounds/Fort_GamePlay_Sounds/Explosives/Grenade_Bounce_Mono_01}}
 {id=WavePlayer_29, properties {soundWaveAssetPtr=/Game/Sounds/Fort_GamePlay_Sounds/Explosives/Grenade_Bounce_Mono_02}}
]
properties {
 attenuationSettings=/Game/Sounds/Attenuation/Fort_Grenade_Bounce_Att
 bOverrideAttenuation=true
 bOverrideConcurrency=true
 maxDistance=4100
 pitchMultiplier=1.1
 soundClassObject=&Fort_WeaponRanged_Projectile_Effects
 attenuationOverrides {
  bAttenuateWithLPF=true
  falloffDistance=3400
  lPFFrequencyAtMax=400
  lPFFrequencyAtMin=10000
  lPFRadiusMax=1500
  lPFRadiusMin=1000
  stereoSpread=2000
  attenuationShapeExtents {x=700}
 }
 concurrencyOverrides {maxCount=3}
}
