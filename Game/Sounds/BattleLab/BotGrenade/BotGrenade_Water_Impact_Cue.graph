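# Sound cue for the Battle Lab bot grenade hitting water.
# A Random node picks among three equally weighted grenade water impact waves shared from the common explosives sounds.
# It plays boosted in the impacts sound class, using a close-range projectile attenuation asset borrowed from the Batarang.
# A cue-level concurrency override prevents new instances from starting once the limit is reached.
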SoundCue /Game/Sounds/BattleLab/BotGrenade/BotGrenade_Water_Impact_Cue
root=Random_0
nodes [
 {id=Random_0, children [WavePlayer_2,WavePlayer_3,WavePlayer_4], properties {weights [1,1,1]}}
 {id=WavePlayer_2, properties {soundWaveAssetPtr=/Game/Sounds/Fort_GamePlay_Sounds/Explosives/Grenade_Impact_Water_03}}
 {id=WavePlayer_3, properties {soundWaveAssetPtr=/Game/Sounds/Fort_GamePlay_Sounds/Explosives/Grenade_Impact_Water_01}}
 {id=WavePlayer_4, properties {soundWaveAssetPtr=/Game/Sounds/Fort_GamePlay_Sounds/Explosives/Grenade_Impact_Water_02}}
]
properties {
 attenuationSettings=/Game/Sounds/Attenuation/Fort_Weapon_Projectile_Close_Athena_Batarang_Att
 bOverrideConcurrency=true
 maxDistance=2600
 soundClassObject=&Fort_Impacts
 volumeMultiplier=3.4272
 attenuationOverrides {falloffDistance=5000, attenuationShapeExtents {x=1000}}
 concurrencyOverrides {resolutionRule=PreventNew}
}
